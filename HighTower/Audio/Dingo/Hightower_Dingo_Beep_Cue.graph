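# Sound cue for Dingo beep sounds in HighTower, built from supply drop idle recordings and assigned to the Fort_Gadgets sound class.
# A Random node picks among six idle wave variants with equal weight.
# A Modulator at the root varies pitch and volume slightly on each play.
# The cue itself raises overall pitch and halves volume, with a very large max distance.
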SoundCue /HighTower/Audio/Dingo/Hightower_Dingo_Beep_Cue
root=Modulator_0
nodes [
 {id=Modulator_0, children [Random_0], properties {pitchMax=1.2, pitchMin=1, volumeMax=1.25, volumeMin=0.9}}
 {id=Random_0, children [WavePlayer_0,WavePlayer_1,WavePlayer_2,WavePlayer_3,WavePlayer_4,WavePlayer_7], properties {weights [1,1,1,1,1,1]}}
 {id=WavePlayer_0, properties {soundWaveAssetPtr=&/SupplyDrop_HighTower_Idle_01}}
 {id=WavePlayer_1, properties {soundWaveAssetPtr=&/SupplyDrop_HighTower_Idle_02}}
 {id=WavePlayer_2, properties {soundWaveAssetPtr=&/SupplyDrop_HighTower_Idle_03}}
 {id=WavePlayer_3, properties {soundWaveAssetPtr=&/SupplyDrop_HighTower_Idle_04}}
 {id=WavePlayer_4, properties {soundWaveAssetPtr=&/SupplyDrop_HighTower_Idle_05}}
 {id=WavePlayer_7, properties {soundWaveAssetPtr=&/SupplyDrop_HighTower_Idle_06}}
]
properties {maxDistance=2097152, pitchMultiplier=1.5, soundClassObject=&Fort_Gadgets, volumeMultiplier=0.5}
