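# Sound cue for Dingo collision impacts in HighTower, routed through the Fort_Gadgets sound class.
# A random node picks one of five equally weighted SupplyDrop HighTower impact waves and feeds it through a modulator at the root.
# The cue plays pitched up and at reduced volume, with a concurrency override that caps how many instances play at once.
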SoundCue /HighTower/Audio/Dingo/Hightower_Dingo_Collision_Cue
root=Modulator_0
nodes [
 {id=Modulator_0, children [Random_0]}
 {id=Random_0, children [WavePlayer_0,WavePlayer_1,WavePlayer_2,WavePlayer_3,WavePlayer_4], properties {weights [1,1,1,1,1]}}
 {id=WavePlayer_0, properties {soundWaveAssetPtr=&/SupplyDrop_HighTower_Impact_01}}
 {id=WavePlayer_1, properties {soundWaveAssetPtr=&/SupplyDrop_HighTower_Impact_02}}
 {id=WavePlayer_2, properties {soundWaveAssetPtr=&/SupplyDrop_HighTower_Impact_03}}
 {id=WavePlayer_3, properties {soundWaveAssetPtr=&/SupplyDrop_HighTower_Impact_04}}
 {id=WavePlayer_4, properties {soundWaveAssetPtr=&/SupplyDrop_HighTower_Impact_06}}
]
properties {
 bOverrideConcurrency=true
 maxDistance=2097152
 pitchMultiplier=1.5
 soundClassObject=&Fort_Gadgets
 volumeMultiplier=0.7
 concurrencyOverrides {maxCount=4}
}
